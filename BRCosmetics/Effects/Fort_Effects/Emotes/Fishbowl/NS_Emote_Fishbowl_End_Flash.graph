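# Niagara system for the ending flash of the Fishbowl emote, drawing short-lived, overbright white sprite flashes.
# Two near-identical sprite emitters use instantaneous bursts at opposite side offsets from the system location, the second smaller and with a loop delay set to stagger it.
# Each sprite starts at a random rotation, spins, is offset by camera offset, fades color and alpha in and out, and shrinks sharply over its latter half.
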NiagaraSystem /BRCosmetics/Effects/Fort_Effects/Emotes/Fishbowl/NS_Emote_Fishbowl_End_Flash
dependencies {usedDataInterfaces [&NiagaraDataInterfaceCurve], usedDynamicInputs [], usedModules [&SystemState]}
emitters {
 End_Flash01 {
  emitterUpdateScript [
   &EmitterState(Life Cycle Mode=::NewEnumerator1, Loop Behavior=::NewEnumerator1, Loop Duration=0.5, UseLoopDelay=true)
   &SpawnBurst_Instantaneous(Use Spawn Probability=false, UseLoopCountLimit=false)
  ]
  particleSpawnScript [
   /Niagara/Modules/Spawn/Initialization/InitializeParticle(
    Color=&MultiplyLinearColor(Linear Color=<rgba 1 1 1 1>, Scale Factor=&MakeVector4_FromFloat(Float=2))
    Lifetime=0.3
    Module.Write Color=true
    Module.Write Lifetime=true
    Module.Write Mass=true
    Module.Write Material Random=true
    Module.Write Position=true
    Module.Write Scale=false
    Module.Write SpriteRotation=true
    Module.Write SpriteSize=true
    Sprite Rotation=&UniformRangedFloat(Minimum=0, Maximum=360)
    Sprite Size=&Vector2DFromFloat(Value=30)
   )
   &CameraOffset(Camera Offset Amount=5) {module=CameraOffset001}
   &SystemLocation(MaskSpawnGroup=false, Offset=<0 -7 0>)
  ]
  particleUpdateScript [
   &ScaleColor(
    Scale Alpha=&FloatFromCurve(FloatCurve=DataInterface() {propertyValues {Curve {keys [<at 0 0>,<at 0.5016 0.9767>,<at 1 0>]}}})
    Scale RGB=&VectorFromFloat(Value=&FloatFromCurve(FloatCurve=DataInterface() {propertyValues {Curve {keys [<at 0 0>,<at 0.5017 1.0042>,<at 1 0>]}}}))
    ScaleA=true
    ScaleRGB=true
   )
   &ParticleState()
   &SpriteRotationRate(Delta Time:NiagaraFloat=$Particles.Lifetime, Rotation Rate=10)
   &SolveForcesAndVelocity(Clamp Velocity=false, Limit Acceleration=false)
   &ScaleSpriteSize(Scale Factor=&Vector2DFromFloat(Value=&FloatFromCurve(FloatCurve=DataInterface() {propertyValues {Curve {keys [
          <at 0 1 RCIM_Cubic>
          {arriveTangent=-0.01, interpMode=RCIM_Cubic, leaveTangent=-0.01, time=0.335, value=1}
          {arriveTangent=-0.0228, interpMode=RCIM_Cubic, leaveTangent=-2.1462, tangentMode=RCTM_Break, time=0.5002, value=0.995}
          {arriveTangent=-1.9908, interpMode=RCIM_Cubic, leaveTangent=-1.9908, time=0.6669, value=0.2877}
          <at 1 0>
         ]}}})))
  ]
  rendererClasses [&NiagaraSpriteRendererProperties]
  renderers [{rendererClass=&NiagaraSpriteRendererProperties}]
 }
 End_Flash02 {
  emitterUpdateScript [
   &EmitterState(Life Cycle Mode=::NewEnumerator1, Loop Behavior=::NewEnumerator1, Loop Delay=0.6, Loop Duration=0.5, UseLoopDelay=true)
   &SpawnBurst_Instantaneous(Use Spawn Probability=false, UseLoopCountLimit=false)
  ]
  particleSpawnScript [
   &InitializeParticle(
    Color=&MultiplyLinearColor(Linear Color=<rgba 1 1 1 1>, Scale Factor=&MakeVector4_FromFloat(Float=2))
    Lifetime=0.3
    Module.Write Color=true
    Module.Write Lifetime=true
    Module.Write Mass=true
    Module.Write Material Random=true
    Module.Write Position=true
    Module.Write Scale=false
    Module.Write SpriteRotation=true
    Module.Write SpriteSize=true
    Sprite Rotation=&UniformRangedFloat(Minimum=0, Maximum=360)
    Sprite Size=&Vector2DFromFloat(Value=15)
   )
   &CameraOffset(Camera Offset Amount=5) {module=CameraOffset001}
   &SystemLocation(MaskSpawnGroup=false, Offset=<0 7 -2>)
  ]
  particleUpdateScript [
   &ScaleColor(
    Scale Alpha=&FloatFromCurve(FloatCurve=DataInterface() {propertyValues {Curve {keys [<at 0 0>,<at 0.5016 0.9767>,<at 1 0>]}}})
    Scale RGB=&VectorFromFloat(Value=&FloatFromCurve(FloatCurve=DataInterface() {propertyValues {Curve {keys [<at 0 0>,<at 0.5017 1.0042>,<at 1 0>]}}}))
    ScaleA=true
    ScaleRGB=true
   )
   &ParticleState()
   &SpriteRotationRate(Delta Time:NiagaraFloat=$Particles.Lifetime, Rotation Rate=10)
   &SolveForcesAndVelocity(Clamp Velocity=false, Limit Acceleration=false)
   &ScaleSpriteSize(Scale Factor=&Vector2DFromFloat(Value=&FloatFromCurve(FloatCurve=DataInterface() {propertyValues {Curve {keys [
          <at 0 1 RCIM_Cubic>
          {arriveTangent=-0.01, interpMode=RCIM_Cubic, leaveTangent=-0.01, time=0.335, value=1}
          {arriveTangent=-0.0228, interpMode=RCIM_Cubic, leaveTangent=-2.1462, tangentMode=RCTM_Break, time=0.5002, value=0.995}
          {arriveTangent=-1.9908, interpMode=RCIM_Cubic, leaveTangent=-1.9908, time=0.6669, value=0.2877}
          <at 1 0>
         ]}}})))
  ]
  rendererClasses [&NiagaraSpriteRendererProperties]
  renderers [{rendererClass=&NiagaraSpriteRendererProperties}]
 }
}
summary {
 systemName=NS_Emote_Fishbowl_End_Flash
 emitters [
  {emitterName=End_Flash01, rendererClasses [&NiagaraSpriteRendererProperties]}
  {emitterName=End_Flash02, rendererClasses [&NiagaraSpriteRendererProperties]}
 ]
 userVariables [
  User.bIsFrontEnd {type=NiagaraBool}
  User.bIsFrontEndPreview {type=NiagaraBool}
  User.IsSwing {type=NiagaraBool}
  User.Return {type=NiagaraBool}
 ]
}
